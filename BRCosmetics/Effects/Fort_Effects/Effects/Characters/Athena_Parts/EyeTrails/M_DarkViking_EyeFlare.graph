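# Translucent unlit particle material for the Dark Viking eye flare, built for Niagara sprites and mesh particles.
# A distortion mask pans in radial coordinates at a dynamic-parameter speed.
# The mask indexes an ice-flame curve atlas, tinted by particle color and HDR-remapped for emissive.
# The mask times a flare burst texture and particle alpha forms opacity, softened by view-angle and depth fades.
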Material /BRCosmetics/Effects/Fort_Effects/Effects/Characters/Athena_Parts/EyeTrails/M_DarkViking_EyeFlare
graph {
 properties {
  bUsedWithMeshParticles=true
  bUsedWithNiagaraSprites=true
  blendMode=BLEND_Translucent
  pixelDepthOffsetMode=PDOM_Legacy
  shadingModel=MSM_Unlit
  usedShadingModels=Unlit
 }
 nodes [
  n1 = ParticleColor()
  n2 = Constant()
  n3 = DynamicParameter() {defaultValue=<rgba -0.25 1 1 1>, paramNames [Panning Speed, Param2, Param3, Param4]}
  n4 = TextureSample(UVs=Panner(
    Coordinate=MaterialFunctionCall() {materialFunction=&VectorToRadialValue}.Radial Coordinates
    Speed=AppendVector(A=@n2, B=@n3.Panning Speed)
   )) {texture=/Game/Global/Textures/Default/Masks_and_Gradients/T_BlasterBeamDistort}
  n5 = Multiply(A=@n4.R, B=TextureSample() {texture=/Game/Effects/Fort_Effects/Materials/T_EFX_Flare_Burst_01}.R)
  DepthFade(
   Opacity=Multiply(
    A=Multiply(A=@n1.A, B=@n5)
    B=MaterialFunctionCall() {materialFunction=&FadeBasedOnViewAngle}.Result
   )
   FadeDistance=ScalarParameter() {defaultValue=32, parameterName=Fade Distance}
  )
  n6 = CurveAtlasRowParameter(CurveTime=Power(Base=@n4.R) {constExponent=0.5}) {
   atlas=&AthenaCurveAtlas01
   curve=/Game/Athena/LinearColorCurves/IceFlamesCurve
   defaultValue=5
   parameterName=Curve
  }
  n7 = Multiply(A=@n6, B=@n1.RGB)
  MaterialFunctionCall(Emissive=MaterialFunctionCall(LDR Input=Clamp(None=@n7), HDR Tint=@n7) {materialFunction=&DeriveHDRfromLDR}.RemappedRange) {materialFunction=&MF_FX_EmissiveAdjustments}
 ]
}
